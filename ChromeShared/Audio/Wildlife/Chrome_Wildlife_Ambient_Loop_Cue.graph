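# Sound cue for a looping Chrome wildlife ambient sound, routed to the Fort_Creatures sound class.
# A looping Wave Player feeds a Modulator that fixes volume at unity, reusing the Chrome vehicle ambient loop wave rather than a wildlife-specific one.
# Spatial behavior comes from the referenced wildlife attenuation asset.
# A concurrency override allows only a few simultaneous instances, stopping the lowest priority beyond that.
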SoundCue /ChromeShared/Audio/Wildlife/Chrome_Wildlife_Ambient_Loop_Cue
root=Modulator_0
nodes [
 {id=Modulator_0, children [WavePlayer_0], properties {volumeMax=1, volumeMin=1}}
 {id=WavePlayer_0, properties {bLooping=true, soundWaveAssetPtr=/ChromeShared/Audio/Vehicles/Chrome_Veh_Ambient_Loop_01}}
]
properties {
 attenuationSettings=&/Chrome_Wildlife_Ambient_Loop_Att
 bOverrideConcurrency=true
 maxDistance=1500
 soundClassObject=/Game/Sounds/FortSoundClassesAndModes/Fort_Creatures
 volumeMultiplier=1
 attenuationOverrides {bAttenuate=false, bSpatialize=false, falloffDistance=2400, attenuationShapeExtents {x=600}}
 concurrencyOverrides {maxCount=3, resolutionRule=StopLowestPriority}
}
