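# Sound cue for the chrome removal effect on wildlife, routed to the creatures sound class.
# A random node picks one of four equally weighted removal waves, then a modulator varies pitch while holding volume steady.
# It references the wildlife heal attenuation asset and boosts overall volume.
# Its own concurrency override caps simultaneous copies at three, stopping the lowest priority and scaling volume by priority.
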SoundCue /ChromeShared/Audio/Wildlife/Chrome_Wildlife_Chrome_Removal_Cue
root=Modulator_0
nodes [
 {id=Modulator_0, children [Random_0], properties {pitchMax=1.2, volumeMax=1, volumeMin=1}}
 {id=Random_0, children [WavePlayer_1,WavePlayer_2,WavePlayer_3,WavePlayer_4], properties {weights [1,1,1,1]}}
 {id=WavePlayer_1, properties {soundWaveAssetPtr=&/Chrome_Creature_Chrome_Removal_01}}
 {id=WavePlayer_2, properties {soundWaveAssetPtr=&/Chrome_Creature_Chrome_Removal_02}}
 {id=WavePlayer_3, properties {soundWaveAssetPtr=&/Chrome_Creature_Chrome_Removal_03}}
 {id=WavePlayer_4, properties {soundWaveAssetPtr=&/Chrome_Creature_Chrome_Removal_04}}
]
properties {
 attenuationSettings=&/Chrome_Wildlife_Heal_Att
 bOverrideConcurrency=true
 maxDistance=12000
 soundClassObject=/Game/Sounds/FortSoundClassesAndModes/Fort_Creatures
 volumeMultiplier=2.5
 attenuationOverrides {bAttenuate=false, bSpatialize=false, falloffDistance=2400, attenuationShapeExtents {x=600}}
 concurrencyOverrides {maxCount=3, resolutionRule=StopLowestPriority, volumeScale=0.7, volumeScaleMode=Priority}
}
