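# Sound cue for the wildlife creature heal sound, routed to the Fort_Creatures sound class with its own heal attenuation asset.
# A Random node picks one of four equally weighted heal waves, and a root Modulator varies its pitch while holding volume steady.
# A concurrency override caps simultaneous plays, stopping the lowest priority and scaling volume by priority, and virtualization is disabled.
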SoundCue /ChromeShared/Audio/Wildlife/Chrome_Wildlife_Heal_Cue
root=Modulator_0
nodes [
 {id=Modulator_0, children [Random_1], properties {pitchMax=1.2, volumeMax=1, volumeMin=1}}
 {
  id=Random_1
  children [WavePlayer_0,WavePlayer_5,WavePlayer_6,WavePlayer_7]
  properties {bShouldExcludeFromBranchCulling=true, weights [1,1,1,1]}
 }
 {id=WavePlayer_0, properties {soundWaveAssetPtr=&/Chrome_Creature_Heal_01}}
 {id=WavePlayer_5, properties {soundWaveAssetPtr=&/Chrome_Creature_Heal_02}}
 {id=WavePlayer_6, properties {soundWaveAssetPtr=&/Chrome_Creature_Heal_03}}
 {id=WavePlayer_7, properties {soundWaveAssetPtr=&/Chrome_Creature_Heal_04}}
]
properties {
 attenuationSettings=&/Chrome_Wildlife_Heal_Att
 bOverrideConcurrency=true
 maxDistance=12000
 priority=0
 soundClassObject=/Game/Sounds/FortSoundClassesAndModes/Fort_Creatures
 virtualizationMode=Disabled
 volumeMultiplier=2.5
 attenuationOverrides {bAttenuate=false, bSpatialize=false, falloffDistance=2400, attenuationShapeExtents {x=600}}
 concurrencyOverrides {maxCount=2, resolutionRule=StopLowestPriority, volumeScale=0.7, volumeScaleMode=Priority}
}
